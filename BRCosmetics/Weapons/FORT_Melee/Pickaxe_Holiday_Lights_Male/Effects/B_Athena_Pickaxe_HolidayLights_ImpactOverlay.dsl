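; Weapon Blueprint for the Holiday Lights pickaxe cosmetic that layers an extra Niagara effect onto melee impacts.
; On the weapon's impact FX event, it spawns the Holiday Lights impact system at the world location and rotation of the already spawned impact FX component.
; The overlay is auto-released and ignores the hit result and surface type, so every surface gets the same effect.
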(event FortWeapon|EventOnPlayImpactFX
  (params (HitResult :type Hit%20Result%20Structure%20%28by%20ref%29) (ImpactPhysicalSurface :type EPhysicalSurface%20Enum :default "SurfaceType_Default") (SpawnedPSC :type FXSystem%20Component%20Object%20Reference))
  (Effects|Components|ParticleSystem|SpawnFXAtLocation "/BRCosmetics/Weapons/FORT_Melee/Pickaxe_Holiday_Lights_Male/Effects/NS_Pickaxe_HolidayLights_Impact.NS_Pickaxe_HolidayLights_Impact" (Transformation|GetWorldLocation SpawnedPSC) (Transformation|GetWorldRotation SpawnedPSC) "1.000000,1.000000,1.000000" true "AutoRelease"))

(fn ConstructionScript
  (|CallParentFunction))
