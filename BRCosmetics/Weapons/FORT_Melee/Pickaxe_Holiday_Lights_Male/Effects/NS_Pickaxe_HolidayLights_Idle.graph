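# Niagara idle effect for the Holiday Lights pickaxe, adding a glowing star sprite and light that follow its skeletal mesh.
# The active StarLights emitter uses a loop-count-gated single burst and re-samples the mesh location every update.
# Normalized time-of-day curves scale light color and sprite alpha, dimming both mid-cycle.
# Disabled SnowGlobe and VolumeTest emitters remain, the former constraining drifting particles to mesh-attached spheres offset per spawn group.
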NiagaraSystem /BRCosmetics/Weapons/FORT_Melee/Pickaxe_Holiday_Lights_Male/Effects/NS_Pickaxe_HolidayLights_Idle
dependencies {
 usedDataInterfaces [&NiagaraDataInterfaceColorCurve, &NiagaraDataInterfaceCurve]
 usedDynamicInputs []
 usedModules [&SystemState]
}
emitters {
 SnowGlobe {
  bEnabled=false
  emitterUpdateScript [
   &EmitterState(Life Cycle Mode=::NewEnumerator1, Loop Duration=1, UseLoopDelay=false)
   &SpawnRate(SpawnRate=40, Use Spawn Probability=false)
  ]
  particleSpawnScript [
   /Niagara/Modules/Spawn/Initialization/InitializeParticle(
    Lifetime=5
    Module.Write Color=true
    Module.Write Lifetime=true
    Module.Write Mass=false
    Module.Write Material Random=true
    Module.Write Position=false
    Module.Write Scale=false
    Module.Write SpriteRotation=false
    Module.Write SpriteSize=true
    Position=&SimulationPosition()
    Sprite Size=&Vector2DFromFloat(Value=&RandomRangeFloat(Minimum=0, Maximum=1))
   )
   &SkeletalMeshLocation(
    Bone Sampling Mode=::NewEnumerator4
    Inherit Velocity=false
    MaskSpawnGroup=false
    Sampled Position Offset=&MaskVectorBySpawnGroup(
     Masked Vector=&RandomRangeVector(Minimum=<-20 -20 0>, Maximum=<20 20 50>)
     Passthrough Vector=&RandomRangeVector(Minimum=<-20 -20 -20>, Maximum=<20 20 20>)
     Spawn Group Mask=2
    )
    Sampled Position Offset Coordinate Space:ENiagaraMeshTransforms=::NewEnumerator1
    Skeletal Mesh=$User.SkeletalMesh
   )
  ]
  particleUpdateScript [
   &ParticleState()
   &SkeletalMeshLocation(
    Apply Sampled Orientation to Mesh Particles=true
    Apply To Particle Position=false
    Bone Sampling Mode=::NewEnumerator4
    Inherit Velocity=false
    Interpolate Mesh Sampling=false
    MaskSpawnGroup=false
    Sampled Position Offset Coordinate Space:ENiagaraMeshTransforms=::NewEnumerator1
    Skeletal Mesh=$User.SkeletalMesh
   ) {module=SkeletalMeshLocation001}
   &AddVelocity(Velocity=&RandomRangeVector(Minimum=<-2 -2 -2>, Maximum=<2 2 2>, Evaluation Type=::NewEnumerator1))
   &Drag(Use Linear Drag=true, Use Rotational Drag=false)
   &SolveForcesAndVelocity(Clamp Velocity=false, Limit Acceleration=false)
   /SetVariables(Particles.OffsetPerSG:Vector3f=&MakeCustomVectorFromBool(
     Bool=&SetBoolByIntComparison(A=&SpawnGroup(), B=0, Comparison Type=::NewEnumerator4)
     True Vector=<2 0.5 0.45>
     False Vector=&MakeCustomVectorFromBool(
      Bool=&SetBoolByIntComparison(A=&SpawnGroup(), B=1, Comparison Type=::NewEnumerator4)
      True Vector=<-2 -0.5 0.45>
      False Vector=&MakeCustomVectorFromBool(
       Bool=&SetBoolByIntComparison(A=&SpawnGroup(), B=2, Comparison Type=::NewEnumerator4)
       True Vector=<21 0.71 0>
       False Vector=<1 0 0>
      )
     )
    )) {isSetParameters=true}
   &self:ConstrainToSphere(
    Area Offset:Vector3f=$Particles.OffsetPerSG
    Sampled Location:Vector3f=$Output.SkeletalMeshLocation001.SampledPosition
    Sampled Rotation:Quat4f=$Output.SkeletalMeshLocation001.SampledRotation
    Scale Sphere:Vector3f=<1 1 1>
    Sphere Radius=15
    Vector to Clamp:NiagaraPosition=$Particles.Position
   ) {module=ConstrainToSphere}
   &Color(Color=&ColorFromCurve(
     ColorCurve=DataInterface() {propertyValues {
       AlphaCurve {keys [<at 0 1 RCIM_Cubic>]}
       BlueCurve {keys [<at 0 0 RCIM_Cubic>,<at 0.4999 0>,<at 1 1 RCIM_Cubic>]}
       GreenCurve {keys [<at 0 0 RCIM_Cubic>,<at 0.4999 1>,<at 1 0 RCIM_Cubic>]}
       RedCurve {keys [<at 0 1 RCIM_Cubic>,<at 0.4999 0>,<at 1 0 RCIM_Cubic>]}
      }}
     CurveIndex=&Multiply_Float(A=0.0025, B=/Niagara/DynamicInputs/TypeConversions/VectorLengthSafe(Vector:Vector3f=$Engine.Owner.Velocity))
    )) {enabled=false}
  ]
  rendererClasses [&NiagaraSpriteRendererProperties]
  renderers [{rendererClass=&NiagaraSpriteRendererProperties}]
 }
 StarLights {
  emitterUpdateScript [
   &EmitterState(Life Cycle Mode=::NewEnumerator1, Loop Duration=1, UseLoopDelay=false)
   &SpawnBurst_Instantaneous(
    Spawn Count=1
    Spawn Probability=&MakeCustomFloatFromBool(
     Bool=&SetBoolByIntComparison(A=$System.LoopCount, B=0, Comparison Type=::NewEnumerator0)
     True Float=0
     False Float=1
    )
    Use Spawn Probability=true
   )
  ]
  particleSpawnScript [&InitializeParticle(
    Color=<rgba 1 0.9202 0.443 0.2>
    Module.Write Color=true
    Module.Write Lifetime=true
    Module.Write Mass=false
    Module.Write Material Random=true
    Module.Write Position=false
    Module.Write Scale=false
    Module.Write SpriteRotation=false
    Module.Write SpriteSize=true
    Position=&SimulationPosition()
    Sprite Size=&Vector2DFromFloat(Value=65)
   )]
  particleUpdateScript [
   &SkeletalMeshLocation(Bone Sampling Mode=::NewEnumerator4, Inherit Velocity=false, MaskSpawnGroup=false, Skeletal Mesh=$User.SkeletalMesh)
   &ParticleState(Kill Particles When Lifetime Has Elapsed=false, Let Infinitely Lived Particles Die When Emitter Deactivates=true)
   &SolveForcesAndVelocity(Clamp Velocity=false, Limit Acceleration=false)
   &FN_Light_v3(
    Color=<rgba 1 0.9225 0.526 1>
    Create Color Attribute=true
    Create Exponent Attribute=true
    Create Radius Attribute=true
    Exponent=2
    Radius=50
    Scale Color:NiagaraFloat=&FloatFromCurve(
     FloatCurve=DataInterface() {propertyValues {Curve {keys [
         {interpMode=RCIM_Cubic, tangentMode=RCTM_User, time=0, value=1}
         {interpMode=RCIM_Cubic, tangentMode=RCTM_User, time=0.5, value=0.5}
         {interpMode=RCIM_Cubic, tangentMode=RCTM_User, time=1, value=1}
        ]}}}
     CurveIndex=Linked() {linkedVariable {name=NPC.FortniteNPC.TimeOfDay_0-1, type=NiagaraFloat}}
    )
   )
   &ScaleColor(
    Scale Alpha=&FloatFromCurve(
     FloatCurve=DataInterface() {propertyValues {Curve {keys [
         {interpMode=RCIM_Cubic, tangentMode=RCTM_User, time=0, value=1}
         {interpMode=RCIM_Cubic, tangentMode=RCTM_User, time=0.5, value=0.2}
         {interpMode=RCIM_Cubic, tangentMode=RCTM_User, time=1, value=1}
        ]}}}
     CurveIndex=Linked() {linkedVariable {name=NPC.FortniteNPC.TimeOfDay_0-1, type=NiagaraFloat}}
    )
    ScaleA=true
    ScaleRGB=false
   )
  ]
  rendererClasses [&NiagaraSpriteRendererProperties, &NiagaraLightRendererProperties]
  renderers [
   {rendererClass=&NiagaraSpriteRendererProperties}
   {rendererClass=&NiagaraLightRendererProperties, rendererIndex=1}
  ]
 }
 VolumeTest {
  bEnabled=false
  emitterUpdateScript [
   &EmitterState(Life Cycle Mode=::NewEnumerator1, Loop Duration=1, UseLoopDelay=false)
   &SpawnRate(SpawnRate=200, Use Spawn Probability=false)
  ]
  particleSpawnScript [&InitializeParticle(
    Lifetime=5
    Module.Write Color=true
    Module.Write Lifetime=true
    Module.Write Mass=true
    Module.Write Material Random=true
    Module.Write Position=false
    Module.Write Scale=false
    Module.Write SpriteRotation=false
    Module.Write SpriteSize=true
    Position=&SimulationPosition()
    Sprite Size=<1 1>
   )]
  particleUpdateScript [
   &ParticleState()
   /SetVariables(Particles.CheckArea:Vector3f=&Add_Vector(
     A=<8 0 0>
     B=&Multiply_Vector(A=<1 0.5 0.6>, B=&RandomVector(Vector Scale=20))
    )) {isSetParameters=true}
   /SetVariables(Particles.CheckArea:Vector3f=&Add_Vector(A=<-22.2411 -14.0594 -7.4818>, B=&RandomRangeVector(Minimum=<-4 -4 -10>, Maximum=<4 4 10>))) {enabled=false, isSetParameters=true}
   &SkeletalMeshLocation(
    Bone / Socket Index=3
    Bone Sampling Mode=::NewEnumerator4
    Inherit Velocity=false
    MaskSpawnGroup=false
    Sampled Position Offset=$Particles.CheckArea
    Sampled Position Offset Coordinate Space:ENiagaraMeshTransforms=::NewEnumerator3
    Skeletal Mesh=$User.SkeletalMesh
   )
   &SolveForcesAndVelocity(Clamp Velocity=false, Limit Acceleration=false)
   &KillParticlesInVolume(
    Kill Shape=::NewEnumerator2
    Origin Offset:Vector3f=<-22.9325 0 0>
    Plane Normal:Vector3f=&RotateVectorByQuaternion(Vector To Rotate=<1.9283 0 1>, Quaternion=$Output.SkeletalMeshLocation.SampledRotation)
    Volume Origin:Vector3f=$Output.SkeletalMeshLocation.SampledPosition
   ) {enabled=false}
  ]
  rendererClasses [&NiagaraSpriteRendererProperties]
  renderers [{rendererClass=&NiagaraSpriteRendererProperties}]
 }
}
summary {
 systemName=NS_Pickaxe_HolidayLights_Idle
 emitters [
  {bEnabled=false, emitterName=SnowGlobe, rendererClasses [&NiagaraSpriteRendererProperties]}
  {bEnabled=false, emitterName=VolumeTest, rendererClasses [&NiagaraSpriteRendererProperties]}
  {emitterName=StarLights, rendererClasses [&NiagaraSpriteRendererProperties, &NiagaraLightRendererProperties]}
 ]
 userVariables [User.SkeletalMesh {
   type=NiagaraDataInterfaceSkeletalMesh
   defaultValue {
    struct=NiagaraExt_VariableValue_DataInterface
    value {dataInterface=&self:NiagaraDataInterfaceSkeletalMesh_1}
   }
  }]
}
