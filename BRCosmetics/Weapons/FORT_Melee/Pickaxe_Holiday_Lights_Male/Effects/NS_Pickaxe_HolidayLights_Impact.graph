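# Impact burst for the Holiday Lights pickaxe, throwing ice, flakes and colored sparkles forward from a small hemisphere.
# One "Ice" emitter fires three instantaneous bursts tagged by spawn group, which drive visibility tags for its mesh and two sprite renderers and vary velocity and gravity.
# Particles drag, spin and scale out over short lives, with pale blue-white flakes and sparkles randomly blended between orange, red and green.
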NiagaraSystem /BRCosmetics/Weapons/FORT_Melee/Pickaxe_Holiday_Lights_Male/Effects/NS_Pickaxe_HolidayLights_Impact
dependencies {usedDataInterfaces [&NiagaraDataInterfaceCurve], usedDynamicInputs [], usedModules [&SystemState]}
emitters {Ice {
  emitterUpdateScript [
   &EmitterState()
   &SpawnBurst_Instantaneous(Spawn Count=5, Use Spawn Probability=false)
   &SpawnBurst_Instantaneous(Spawn Count=&RandomRangeInt(Minimum=2, Maximum=5), Spawn Group=1, Use Spawn Probability=false) {module=SpawnBurst_Instantaneous001}
   &SpawnBurst_Instantaneous(Spawn Count=&RandomRangeInt(Minimum=10, Maximum=15), Spawn Group=2, Use Spawn Probability=false) {module=SpawnBurst_Instantaneous002}
  ]
  particleSpawnScript [
   /SetVariables(
    Particles.HideFlakes:NiagaraInt32=&self:SpawnGroupToVisibilityTag(Spawn Group:NiagaraInt32=&SpawnGroup(), Visibility Offset:NiagaraInt32=-1)
    Particles.HideSparkles:NiagaraInt32=&self:SpawnGroupToVisibilityTag(Spawn Group:NiagaraInt32=&SpawnGroup(), Visibility Offset:NiagaraInt32=-2)
    Particles.Rand01:NiagaraFloat=&UniformAOrBFloat(A=0)
    Particles.Rand02:NiagaraFloat=&UniformAOrBFloat(A=0)
    Particles.SpawnGroup=&SpawnGroup()
   ) {isSetParameters=true}
   /Niagara/Modules/Spawn/Initialization/InitializeParticle(
    Color=&MakeLinearColorFromVectorAndFloat(
     "Vector (RGB)"=&RandomRangeVector(Minimum=<0 1 1>, Maximum=<1 1 1>)
     "Float (Alpha)"=&RandomRangeFloat(Minimum=1, Maximum=2)
    )
    Lifetime=&Power(a=&RandomRangeFloat(Minimum=0.4, Maximum=0.9), b=2)
    Mesh Scale=&RandomRangeVector(Minimum=<0.05 0.05 0.05>, Maximum=<0.3 0.3 0.3>)
    Module.Write Color=true
    Module.Write Lifetime=true
    Module.Write Mass=true
    Module.Write Material Random=true
    Module.Write Position=true
    Module.Write Scale=true
    Module.Write SpriteRotation=true
    Module.Write SpriteSize=true
    Sprite Rotation=&RandomRangeFloat(Minimum=0, Maximum=360)
    Sprite Size=&Vector2DFromFloat(Value=&RandomRangeFloat(Minimum=5, Maximum=15))
   )
   &DynamicMaterialParameters(
    Index 0 Param 2=&RandomRangeFloat(Minimum=30, Maximum=50)
    Module.Param0WriteEnabled=false
    Module.Param1WriteEnabled=true
    Module.Param2WriteEnabled=false
    Module.Param3WriteEnabled=true
   ) {module=DynamicMaterialParameters001}
   &SphereLocation(
    Hemisphere X=true
    MaskSpawnGroup=false
    Override Seed=false
    Sphere Radius=3
    Surface Only=false
    UseNonUniformScale=false
    UseOffset=false
    UseSphereOrientation=false
   )
   &AddVelocityFromPoint(
    Module.Use Falloff Distance=false
    Velocity Strength=&MakeCustomFloatFromBool(
     Bool=&SetBoolByIntComparison(A=$Particles.SpawnGroup, B=2, Comparison Type=::NewEnumerator4)
     True Float=&RandomRangeFloat(Minimum=50, Maximum=500)
     False Float=&RandomRangeFloat(Minimum=0, Maximum=250)
    )
   )
   &AddVelocity(Velocity=&RandomRangeVector(Minimum=<30 0 0>, Maximum=<300 0 0>))
   &AddVelocity(Coordinate Space=::World, Velocity=&RandomRangeVector(Minimum=<0 0 10>, Maximum=<0 0 50>)) {module=AddVelocity001}
   &InitialMeshOrientation(
    Rotation=&RandomRangeVector(Minimum=<0 0 0>, Maximum=<1 1 1>)
    Rotation Coordinate Space=::NewEnumerator0
    Use Orientation Vector=false
    Use Rotation Vector=true
   )
  ]
  particleUpdateScript [
   &GravityForce(Gravity=&MakeVector(
     X=0
     Y=0
     Z=&FloatFromCurve(
      FloatCurve=DataInterface() {propertyValues {Curve {keys [{arriveTangent=0.0468, interpMode=RCIM_Cubic, time=0.25, value=0}, <at 0.5 1 RCIM_Cubic>]}}}
      Scale Curve=&MakeCustomFloatFromBool(
       Bool=&SetBoolByIntComparison(A=$Particles.SpawnGroup, B=0, Comparison Type=::NewEnumerator4)
       True Float=&RandomRangeFloat(Minimum=-2000, Maximum=-1000)
       False Float=&RandomRangeFloat(Minimum=-1000, Maximum=-10)
      )
     )
    ))
   &Drag(Drag=4, Use Linear Drag=true, Use Rotational Drag=false)
   &SolveForcesAndVelocity(Clamp Velocity=false, Limit Acceleration=false)
   &ParticleState()
   &DynamicMaterialParameters(
    Index 0 Param 1=&FloatFromCurve(FloatCurve=DataInterface() {propertyValues {Curve {keys [<at 0 10>,<at 1 0>]}}})
    Index 0 Param 2=&RandomRangeFloat(Minimum=30, Maximum=50, Evaluation Type=::NewEnumerator1)
    Index 0 Param 4=&FloatFromCurve(FloatCurve=DataInterface() {propertyValues {Curve {keys [<at 0 0.2>,<at 1 0>]}}})
    Module.Param0WriteEnabled=true
    Module.Param1WriteEnabled=false
    Module.Param2WriteEnabled=true
    Module.Param3WriteEnabled=true
   )
   &ScaleColor(
    Scale Alpha=&FloatFromCurve(FloatCurve=DataInterface() {propertyValues {Curve {keys [
         {arriveTangent=4.2631, interpMode=RCIM_Cubic, leaveTangent=4.2631, tangentMode=RCTM_User, time=0, value=0}
         <at 0.3504 1 RCIM_Cubic>
         {arriveTangent=-0.1318, interpMode=RCIM_Cubic, leaveTangent=-0.1318, tangentMode=RCTM_User, time=1, value=0}
        ]}}})
    ScaleA=true
    ScaleRGB=true
   )
   &ScaleMeshSize(Scale Factor=&VectorFromFloat(Value=&FloatFromCurve(FloatCurve=DataInterface() {propertyValues {Curve {keys [
          {interpMode=RCIM_Cubic, tangentMode=RCTM_User, time=0, value=1}
          {arriveTangent=-0.0578, interpMode=RCIM_Cubic, leaveTangent=-0.0578, tangentMode=RCTM_User, time=0.5, value=1}
          {arriveTangent=-2.6249, interpMode=RCIM_Cubic, leaveTangent=-2.6249, tangentMode=RCTM_User, time=1, value=0}
         ]}}})))
   &ScaleSpriteSize(Scale Factor=&Vector2DFromFloat(Value=&FloatFromCurve(
      FloatCurve=DataInterface() {propertyValues {Curve {keys [
          {arriveTangent=4.9073, interpMode=RCIM_Cubic, leaveTangent=4.9073, tangentMode=RCTM_User, time=0, value=0}
          <at 0.2737 1 RCIM_Cubic>
          {arriveTangent=-2.1114, interpMode=RCIM_Cubic, leaveTangent=-2.1114, tangentMode=RCTM_User, time=1, value=0}
         ]}}}
      CurveIndex=&MakeCustomFloatFromBool(
       Bool=&SetBoolByIntComparison(A=$Particles.SpawnGroup, B=3, Comparison Type=::NewEnumerator4)
       True Float=&Multiply_Float(A=$Particles.NormalizedAge, B=2)
       False Float=$Particles.NormalizedAge
      )
      Scale Curve=&MakeCustomFloatFromBool(
       Bool=&SetBoolByIntComparison(A=$Particles.SpawnGroup, B=3, Comparison Type=::NewEnumerator4)
       True Float=&RandomRangeFloat(Minimum=1, Maximum=2)
       False Float=1
      )
     )))
   &UpdateMeshOrientation(
    Rotation Rate=&Multiply_Float(A=0.02, B=/Niagara/DynamicInputs/TypeConversions/VectorLengthSafe(Vector:Vector3f=$Particles.Velocity))
    Rotation Vector=&RandomRangeVector(Minimum=<0 0 0>, Maximum=<1 1 1>)
   )
   /SetVariables(
    Particles.FlakeColor:LinearColor=&RandomRangeLinearColor(Minimum=<rgba 0.2266 0.2929 0.724 5>, Maximum=<rgba 1 1 1 5>)
    Particles.SparklesColor:LinearColor=&HueShiftLinearColor(
     Hue=&RandomRangeFloat(Minimum=0, Maximum=0)
     Saturation=&RandomRangeFloat(Minimum=0.6, Maximum=1)
     Value=&FloatFromCurve(
      FloatCurve=DataInterface() {propertyValues {Curve {keys [{arriveTangent=3.1648, interpMode=RCIM_Cubic, leaveTangent=3.1648, tangentMode=RCTM_User, time=0, value=1}, <at 0.5 0>]}}}
      Scale Curve=&RandomRangeFloat(Minimum=0.5, Maximum=10)
     )
     Linear Color=&LerpLinearColors(
      StartColor=<rgba 10 4.1667 0 1>
      EndColor=&LerpLinearColors(StartColor=<rgba 10 0 0 1>, EndColor=<rgba 0 10 0 1>, LerpFactor=$Particles.Rand01)
      LerpFactor=$Particles.Rand02
     )
    )
    Particles.SpriteRotation:NiagaraFloat=&Add_Float(
     A=$Particles.Initial.SpriteRotation
     B=&Multiply_Float(
      A=&Lerp_Float(A=0, B=200, Alpha=$Particles.NormalizedAge)
      B=&RandomRangeFloat(Minimum=-1, Maximum=1)
     )
    )
   ) {isSetParameters=true}
  ]
  rendererClasses [&NiagaraMeshRendererProperties, &NiagaraSpriteRendererProperties]
  renderers [
   {rendererClass=&NiagaraMeshRendererProperties, rendererIndex=0}
   {rendererClass=&NiagaraSpriteRendererProperties, rendererIndex=1}
   {rendererClass=&NiagaraSpriteRendererProperties, rendererIndex=2}
  ]
 }}
summary {
 systemName=NS_Pickaxe_HolidayLights_Impact
 emitters [{emitterName=Ice, rendererClasses [&NiagaraMeshRendererProperties, &NiagaraSpriteRendererProperties]}]
 userVariables [
  User.bIsFrontEndPreview {type=NiagaraBool}
  User.FrontendSpawn {type=NiagaraBool}
 ]
}
