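# Niagara swing trail for the holiday lights pickaxe, pairing snow sprites with ribbons sampled from the weapon's skeletal mesh.
# Snow spawns around sampled mesh points, inherits capped velocity, then drifts under drag and gravity.
# Its alternate position, size and orange/red/green sparkle color attributes suggest a second sprite layer.
# SnowTrail builds three fading ribbons indexed to bones or sockets, using per-ribbon masks to separate center and edge widths across two renderers.
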NiagaraSystem /BRCosmetics/Weapons/FORT_Melee/Pickaxe_Holiday_Lights_Male/Effects/NS_Pickaxe_HolidayLights_Trail
dependencies {
 usedDataInterfaces [
  &NiagaraDataInterfaceArrayFloat
  &NiagaraDataInterfaceVectorCurve
  &NiagaraDataInterfaceColorCurve
  &NiagaraDataInterfaceCurve
 ]
 usedDynamicInputs []
 usedModules [&SystemState]
}
emitters {
 Snow {
  emitterUpdateScript [
   &EmitterState()
   &SpawnRate(Spawn Probability=0.75, SpawnRate=100, Use Spawn Probability=true)
  ]
  particleSpawnScript [
   /Niagara/Modules/Spawn/Initialization/InitializeParticle(
    Lifetime=&RandomRangeFloat(Minimum=0.5, Maximum=1.25)
    Module.Write Color=true
    Module.Write Lifetime=true
    Module.Write Mass=false
    Module.Write Material Random=true
    Module.Write Position=false
    Module.Write Scale=false
    Module.Write SpriteRotation=true
    Module.Write SpriteSize=true
    Sprite Rotation=&RandomRangeFloat(Minimum=-180, Maximum=180)
    Sprite Size=&Vector2DFromFloat(Value=&RandomRangeFloat(Minimum=1.5, Maximum=2.6))
   )
   /SetVariables(Particles.SpriteSizeAlt:Vector2f=&Vector2DFromFloat(Value=&RandomRangeFloat(Minimum=3, Maximum=10))) {isSetParameters=true}
   /SetVariables(
    Particles.Rand01:NiagaraFloat=&UniformAOrBFloat(A=0)
    Particles.Rand02:NiagaraFloat=&UniformAOrBFloat(A=0)
    Particles.SpawnGroup=&SpawnGroup()
   ) {isSetParameters=true}
   &SkeletalMeshLocation(
    Bone Sampling Mode=::NewEnumerator4
    Inherit Velocity=false
    MaskSpawnGroup=false
    Sampled Position Offset=&RandomRangeVector(Minimum=<0 0 -25>, Maximum=<0 0 0>)
    Sampled Position Offset Coordinate Space:ENiagaraMeshTransforms=::NewEnumerator3
    Skeletal Mesh=$User.SkeletalMesh
   )
   &SphereLocation(
    MaskSpawnGroup=false
    Override Seed=false
    Sphere Radius=20
    Surface Only=false
    UseNonUniformScale=false
    UseOffset=false
    UseSphereOrientation=false
   )
   &AddVelocityInCone(Module.Use Velocity Falloff On Cone Axis=false, Velocity Strength=70)
   &InheritVelocity(Module.Limit Speed=true, Velocity Limit=200)
  ]
  particleUpdateScript [
   &ParticleState()
   &Drag(Drag=2, Use Linear Drag=true, Use Rotational Drag=false)
   &GravityForce(Gravity=<0 0 -150>)
   &SolveForcesAndVelocity(Clamp Velocity=false, Limit Acceleration=false)
   /SetVariables(
    Particles.PositionAlt:Vector3f=&Add_Vector(
     A=&Lerp_Vector(
      A=Linked() {linkedVariable {name=Particles.Initial.Position, type=NiagaraPosition}}
      B=Linked() {linkedVariable {name=Particles.Position, type=NiagaraPosition}}
      Alpha=0.5
     )
     B=&Lerp_Vector(
      A=<0 0 0>
      B=&RandomVector(Vector Scale=&RandomRangeFloat(Minimum=0, Maximum=20))
      Alpha=$Particles.NormalizedAge
     )
    )
    Particles.SparklesColor:LinearColor=&HueShiftLinearColor(
     Hue=&RandomRangeFloat(Minimum=0, Maximum=0)
     Saturation=&RandomRangeFloat(Minimum=0.6, Maximum=1)
     Value=&FloatFromCurve(
      FloatCurve=DataInterface() {propertyValues {Curve {keys [{arriveTangent=3.1648, interpMode=RCIM_Cubic, leaveTangent=3.1648, tangentMode=RCTM_User, time=0, value=1}, <at 0.5 0>]}}}
      Scale Curve=&RandomRangeFloat(Minimum=0.5, Maximum=10)
     )
     Linear Color=&LerpLinearColors(
      StartColor=<rgba 10 4.1667 0 1>
      EndColor=&LerpLinearColors(StartColor=<rgba 10 0 0 1>, EndColor=<rgba 0 10 0 1>, LerpFactor=$Particles.Rand01)
      LerpFactor=$Particles.Rand02
     )
    )
    Particles.SpriteRotation:NiagaraFloat=&Add_Float(
     A=$Particles.Initial.SpriteRotation
     B=&Lerp_Float(A=0, B=&RandomRangeFloat(Minimum=-200, Maximum=200), Alpha=$Particles.NormalizedAge)
    )
   ) {isSetParameters=true}
   &ScaleSpriteSize(Scale Factor=&Vector2DFromFloat(Value=&FloatFromCurve(FloatCurve=DataInterface() {propertyValues {Curve {keys [{interpMode=RCIM_Cubic, tangentMode=RCTM_User, time=0.6, value=1}, <at 1 0>]}}})))
  ]
  rendererClasses [&NiagaraSpriteRendererProperties]
  renderers [
   {rendererClass=&NiagaraSpriteRendererProperties}
   {rendererClass=&NiagaraSpriteRendererProperties, rendererIndex=1}
  ]
 }
 SnowTrail {
  emitterUpdateScript [
   &EmitterState()
   /Game/Effects/Niagara/Modules/Ribbons/FN_SpawnRibbon_v3(Number of Ribbons=3, Skeleton Mesh:NiagaraDataInterfaceSkeletalMesh=$User.SkeletalMesh, SpawnRate=20)
  ]
  particleSpawnScript [
   /Game/Effects/Niagara/Modules/Ribbons/FN_InitializeRibbon_v5(Lifetime=0.3, Offset Axis:ENiagaraOrientationAxis=::ZAxis, Ribbon Width=50, Seed=10)
   /SetVariables(
    Particles.HideCenter:NiagaraFloat=&SelectFloatFromArray(Direct Array Index=$Particles.SpawnGroup, Float Selection Array=DataInterface() {propertyValues {FloatData [0,1,1]}})
    Particles.HideEdge:NiagaraFloat=&SelectFloatFromArray(Direct Array Index=$Particles.SpawnGroup, Float Selection Array=DataInterface() {propertyValues {FloatData [1,0,0]}})
   ) {isSetParameters=true}
   &SkeletalMeshLocation(
    Bone / Socket Index=$Particles.SpawnGroup
    Bone Sampling Mode=::NewEnumerator4
    Inherit Velocity=false
    Interpolate Mesh Sampling=false
    MaskSpawnGroup=false
    Skeletal Mesh=$User.SkeletalMesh
   )
   /Game/Effects/Niagara/Modules/Ribbons/FN_UpdateRibbon_v5()
  ]
  particleUpdateScript [
   &ParticleState()
   &ScaleColor(
    Scale Mode=::NewEnumerator1
    Scale RGBA=&Vector4FromCurve(Vector4Curve=DataInterface() {propertyValues {
       AlphaCurve {keys [<at 0 1>,<at 0.1 2>,<at 0.32 0.21>,<at 1 0>]}
       BlueCurve {keys [<at 0 3>,<at 0.1508 0>,<at 1 0>]}
       GreenCurve {keys [<at 0 3>,<at 0.1508 0.6167>,<at 1 0.0219>]}
       RedCurve {keys [<at 0 3>,<at 0.1508 1>,<at 1 0.101>]}
      }})
    ScaleRGBA=true
   )
   /SetVariables(
    Particles.ColorAlt:LinearColor=&MakeLinearColorFromVectorAndFloat(
     "Vector (RGB)"=&VectorFromCurve(VectorCurve=DataInterface() {propertyValues {
        XCurve {keys [
          {arriveTangent=-0.8116, interpMode=RCIM_Cubic, leaveTangent=-0.8116, tangentMode=RCTM_User, time=0, value=0.25}
          {arriveTangent=-0.4144, interpMode=RCIM_Cubic, leaveTangent=-0.4144, tangentMode=RCTM_User, time=0.35, value=-0.0044}
          <at 0.5 0>
         ]}
        YCurve {keys [
          {arriveTangent=-6.2182, interpMode=RCIM_Cubic, leaveTangent=-6.2182, tangentMode=RCTM_User, time=0, value=1}
          {arriveTangent=-0.2878, interpMode=RCIM_Cubic, leaveTangent=-0.2878, tangentMode=RCTM_User, time=0.0952, value=0.2864}
          <at 0.5 0>
         ]}
        ZCurve {keys [<at 0 1>,<at 0.5 0>]}
       }})
     "Float (Alpha)"=&FloatFromCurve(FloatCurve=DataInterface() {propertyValues {Curve {keys [{arriveTangent=0.1656, interpMode=RCIM_Cubic, leaveTangent=0.1656, tangentMode=RCTM_User, time=0, value=0.9773}, <at 0.5 0 RCIM_Cubic>]}}})
    )
    Particles.RibbonWidthAlt:NiagaraFloat=&Multiply_Float(A=0.3, B=&Multiply_Float(A=$Particles.HideCenter, B=$Particles.Initial.RibbonWidth))
   ) {isSetParameters=true}
   &ScaleRibbonWidth(Ribbon Width Scale=&FloatFromCurve(
     FloatCurve=DataInterface() {propertyValues {Curve {keys [
         {arriveTangent=6.137, interpMode=RCIM_Cubic, leaveTangent=6.137, tangentMode=RCTM_User, time=0, value=0.5}
         {arriveTangent=-3.1746, interpMode=RCIM_Cubic, leaveTangent=-3.1746, tangentMode=RCTM_User, time=0.13, value=1}
         {arriveTangent=-0.6416, interpMode=RCIM_Cubic, leaveTangent=-0.6416, tangentMode=RCTM_User, time=0.2757, value=0.324}
         {arriveTangent=-0.2837, interpMode=RCIM_Cubic, leaveTangent=-0.2837, tangentMode=RCTM_User, time=1, value=0}
        ]}}}
     Scale Curve=$Particles.HideEdge
    ))
   &DynamicMaterialParameters(
    Index 0 Param 1=4
    Index 0 Param 2=&RandomRangeFloat(Minimum=0, Maximum=1)
    Index 0 Param 3=1
    Index 0 Param 4=&FloatFromCurve(FloatCurve=DataInterface() {propertyValues {Curve {keys [<at 0 1>,<at 1 0>]}}})
    Module.Param0WriteEnabled=true
    Module.Param1WriteEnabled=true
    Module.Param2WriteEnabled=true
    Module.Param3WriteEnabled=true
   )
  ]
  rendererClasses [&NiagaraRibbonRendererProperties]
  renderers [
   {rendererClass=&NiagaraRibbonRendererProperties}
   {rendererClass=&NiagaraRibbonRendererProperties, rendererIndex=1}
  ]
 }
}
summary {
 systemName=NS_Pickaxe_HolidayLights_Trail
 emitters [
  {emitterName=SnowTrail, rendererClasses [&NiagaraRibbonRendererProperties]}
  {emitterName=Snow, rendererClasses [&NiagaraSpriteRendererProperties]}
 ]
 userVariables [
  User.bIsFrontEndPreview {type=NiagaraBool}
  User.RibbonIDOffset {type=NiagaraInt32}
  User.SkeletalMesh {
   type=NiagaraDataInterfaceSkeletalMesh
   defaultValue {
    struct=NiagaraExt_VariableValue_DataInterface
    value {dataInterface=&self:NiagaraDataInterfaceSkeletalMesh_0}
   }
  }
 ]
}
