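# Material function that builds glass material attributes for the Holiday Lights pickaxe, using named reroutes to organize each channel.
# Fresnel terms drive metallic and opacity, with thickness and inner/outer transparency parameters shaping the falloff.
# A static switch selects optional texture opacity from a channel mask, and a TechArtMask input is exposed.
# Emissive is a faint vertex-normal Fresnel rim scaled by a blend toward the CurveColor_Time_of_Day function.
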MaterialFunction /BRCosmetics/Weapons/FORT_Melee/Pickaxe_Holiday_Lights_Male/Materials/MF_HolidayLights
graph {nodes [
  FunctionOutput(None=NamedRerouteUsage())
  n1 = Power(Base=Fresnel(Normal=VertexNormalWS()) {exponent=2.04}, Exp=Constant() {r=1.8236})
  n2 = Constant() {r=0.25}
  n3 = MaterialFunctionCall() {materialFunction=/Game/Effects/Fort_Effects/Materials/_Functions/CurveColor_Time_of_Day}
  NamedRerouteDeclaration(None=Multiply(A=Saturate(None=@n1), B=LinearInterpolate(A=@n2, B=@n3.Result) {constAlpha=0.6}) {constB=0.1}) {name=GlassEmissive, nodeColor=<rgba 0 0.9062 1 1>}
  NamedRerouteDeclaration(None=LinearInterpolate() {constA=0.5}) {name=GlassSpecular, nodeColor=<rgba 0 1 0.0625 1>}
  NamedRerouteDeclaration(None=LinearInterpolate(
    A=ScalarParameter() {defaultValue=0.8255, group=IcyScarf, parameterName=GlassMetallicMin}
    B=ScalarParameter() {defaultValue=0.05, group=IcyScarf, parameterName=GlassMetallicMax}
    Alpha=Fresnel() {exponent=1}
   )) {name=GlassMetallic, nodeColor=<rgba 0 1 0.0625 1>}
  NamedRerouteDeclaration(None=ScalarParameter() {defaultValue=0.2, group=IcyScarf, parameterName=GlassRoughnessMax}) {name=GlassRoughness, nodeColor=<rgba 0 1 0.0625 1>}
  n4 = NamedRerouteUsage()
  n5 = SetMaterialAttributes!(
   MaterialAttributes
   BaseColor=VectorParameter() {parameterName=GlassBaseColor, defaultValue=<rgba 0.5089 0.4233 0.4851 1>}.RGB
   Opacity=@n4
   OpacityMask=@n4
   EmissiveColor=NamedRerouteUsage()
   Metallic=NamedRerouteUsage()
   Specular=NamedRerouteUsage()
   Roughness=NamedRerouteUsage()
   Normal
   Tangent=Constant3Vector() {constant=<rgba 1 0 0 0>}
  ) {attributeSetTypes [
    #1
    #2
    #3
    #4
    #5
    #6
    #7
    #8
    #9
   ]}
  NamedRerouteDeclaration(None=@n5) {name=GlassMaterial, nodeColor=<rgba 0.8516 0 1 1>}
  n6 = Multiply(
   A=Fresnel(ExponentIn=ScalarParameter() {defaultValue=1, group=IcyScarf, parameterName=FresnelExponentTransparency, sortPriority=30})
   B=ScalarParameter() {defaultValue=1, group=IcyScarf, parameterName=GlassThickness, sortPriority=31}
  )
  n7 = LinearInterpolate(
   A=ScalarParameter() {group=IcyScarf, parameterName=InnerTransparency}
   B=ScalarParameter() {defaultValue=0.85, group=IcyScarf, parameterName=OuterTransparency, sortPriority=33}
   Alpha=NamedRerouteDeclaration(None=Saturate(None=@n6)) {name=GlassFresnel, nodeColor=<rgba 1 0.1875 0 1>}
  )
  NamedRerouteDeclaration(None=@n7) {name=GlassFresnelTransparency, nodeColor=<rgba 0 1 0.9531 1>}
  NamedRerouteDeclaration(None=FunctionInput() {bUsePreviewValueAsDefault=true, inputName=TechArtMask, previewValue=<1 1 1 1>}) {name=TechArtMask, nodeColor=<rgba 0.8047 0 1 1>}
  NamedRerouteDeclaration(None=ChannelMaskParameter(None=NamedRerouteUsage()) {group=00_Glass, parameterName=TextureOpacityChannel, sortPriority=36}.R) {name=TextureOpactiy, nodeColor=<rgba 1 0.2578 0 1>}
  n8 = NamedRerouteUsage()
  n9 = Multiply(A=@n8, B=ScalarParameter() {defaultValue=1, group=IcyScarf, parameterName=TextureOpacityAdd, sortPriority=34})
  n10 = NamedRerouteUsage()
  n11 = LinearInterpolate(A=Saturate(None=Add(A=@n9, B=@n10)), B=@n8, Alpha=ScalarParameter() {group=IcyScarf, parameterName=TextureOpacityBlend, sortPriority=35})
  n12 = NamedRerouteUsage()
  NamedRerouteDeclaration(None=Add(A=Saturate(None=@n7), B=StaticSwitchParameter(True=@n11, False=@n12) {group=IcyScarf, parameterName=UseTextureOpacity, sortPriority=36})) {name=GlassOpacity, nodeColor=<rgba 0.9922 0 1 1>}
 ]}
