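# Additive, two-sided lit glass overlay material for the male Holiday Lights pickaxe, built between the CharacterParent 2023 PreFX and PostFX functions.
# A contrast-powered GlassColor parameter, not a texture, drives base color tinted by a time-of-day curve, and opacity clamped to a floor and scaled by FrostOpacity.
# An inverted, saturated TwoSidedSign masks opacity to back faces only, while metallic and roughness are constants.
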Material /BRCosmetics/Weapons/FORT_Melee/Pickaxe_Holiday_Lights_Male/Materials/M_HolidayLightsMale_GlassOverlay
graph {
 properties {
  bAutomaticallySetUsageInEditor=false
  bScreenSpaceReflections=true
  bUseMaterialAttributes=true
  bUsedWithMeshDeformer=true
  bUsedWithSkeletalMesh=true
  blendMode=BLEND_Additive
  translucencyLightingMode=TLM_Surface
  twoSided=true
  usedShadingModels=Default Lit
 }
 nodes [
  n1 = MaterialFunctionCall() {materialFunction=&MF_CharacterParent_2023_PreFX}
  n2 = SetMaterialAttributes!(
   MaterialAttributes=GetMaterialAttributes(None=NamedRerouteDeclaration(None=@n1.Result) {name=InputMA, nodeColor=<rgba 0 1 0.6719 1>}) {attributeGetTypes [#1]}.MaterialAttributes
   Opacity=NamedRerouteUsage()
   Normal
   Base Color=NamedRerouteUsage()
   Metallic=Constant() {r=0.15}
   Roughness=Constant() {r=0.304}
  ) {attributeSetTypes [
    #2
    #3
    #1
    #4
    #5
   ]}
  MaterialFunctionCall(Material=@n2, M=@n1.M, Pre Skinned Local Position=@n1.M, Distance Blend=@n1.M, Fresnel=@n1.M) {materialFunction=&MF_CharacterParent_2023_PostFX}
  n3 = Power(
   Base=VectorParameter() {parameterName=GlassColor, defaultValue=<rgba 0.1572 0.1573 0.1826 1>}.RGB
   Exp=ScalarParameter() {defaultValue=1.6992, group=IcyScarf, parameterName=FrostContrast}
  ) {constExponent=4}
  n4 = Max(A=Multiply(A=@n3, B=Constant() {r=0.8}) {constB=0.5}, B=Constant() {r=0.3})
  n5 = OneMinus(None=Saturate(None=TwoSidedSign()))
  NamedRerouteDeclaration(None=Multiply(A=Multiply(A=@n4, B=@n5), B=ScalarParameter() {defaultValue=0.586, group=IcyScarf, parameterName=FrostOpacity}) {constB=0.75}) {name=Opacity, nodeColor=<rgba 1 0 0.4453 1>}
  n6 = Constant() {r=0.8}
  n7 = MaterialFunctionCall() {materialFunction=/Game/Effects/Fort_Effects/Materials/_Functions/CurveColor_Time_of_Day}
  NamedRerouteDeclaration(None=Multiply(A=Add(A=@n3, B=@n6), B=Multiply(A=@n7.Result) {constB=0.35})) {name=BaseColor, nodeColor=<rgba 0 0.7656 1 1>}
 ]
}
