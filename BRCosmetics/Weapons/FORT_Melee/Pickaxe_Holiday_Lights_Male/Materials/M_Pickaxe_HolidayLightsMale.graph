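# Skeletal mesh material for the male Holiday Lights pickaxe, layering twinkling bulbs and wires over MF_FN_Weapon_MASTER.
# Vertex color red both masks bulbs and offsets their blink phase, with two pre-skinned-position sphere masks overriding the phase locally.
# A panned, wrapped phase pulses bulb emissive before hue shift and desaturation; MinLight is unconnected.
# Bulb and texture-masked wire surface overrides then feed MF_TechArtOverrides_accessories.
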Material /BRCosmetics/Weapons/FORT_Melee/Pickaxe_Holiday_Lights_Male/Materials/M_Pickaxe_HolidayLightsMale
graph {
 properties {
  bUseMaterialAttributes=true
  bUsedWithMeshDeformer=true
  bUsedWithSkeletalMesh=true
  pixelDepthOffsetMode=PDOM_Legacy
  usedShadingModels=Default Lit
 }
 nodes [
  n1 = VertexColor()
  n2 = VertexInterpolator(VS=PreSkinnedPosition())
  n3 = LinearInterpolate(
   A=@n1.R
   B=Constant() {r=0.1}
   Alpha=SphereMask(A=@n2.PS, B=Constant3Vector() {constant=<rgba 1 6 -53 0>}, Radius=Constant() {r=2}, Hardness=Constant() {r=1})
  )
  n4 = Constant() {r=0.5}
  n5 = SphereMask(A=@n2.PS, B=Constant3Vector() {constant=<rgba 8 10 34 0>}, Radius=Constant() {r=5}, Hardness=Constant() {r=1})
  n6 = Panner(
   Coordinate=Multiply(A=LinearInterpolate(A=@n3, B=@n4, Alpha=@n5), B=ScalarParameter() {defaultValue=1, group=TwinklingLights, parameterName=LightCount})
   Speed=ScalarParameter() {defaultValue=-1, group=TwinklingLights, parameterName=Speed}
  ) {speedY=1}
  n7 = SphereMask(
   A=Frac(None=ComponentMask(None=@n6) {r=true})
   B=Constant() {r=0.5}
   Radius=ScalarParameter() {defaultValue=0.2, group=TwinklingLights, parameterName=LightRadius}
   Hardness=ScalarParameter() {defaultValue=1, group=TwinklingLights, parameterName=LightHardness}
  ) {attenuationRadius=0.5, hardnessPercent=0}
  Add(A=@n7, B=ScalarParameter() {group=TwinklingLights, parameterName=MinLight})
  n8 = GetMaterialAttributes(None=MaterialFunctionCall!(In) {materialFunction=/Game/Packages/Fortress_SharedMaterials/Base_Material_Functions/BaseCustomizationFunctions/MF_FN_Weapon_MASTER}.Material Result) {attributeGetTypes [
    #1
    #2
    #3
    #4
    #5
   ]}
  n9 = ScalarParameter() {parameterName=EmissiveHue}
  n10 = Saturate(None=@n7)
  n11 = ScalarParameter() {defaultValue=2, group=TwinklingLights, parameterName=BulbsBrightness}
  MaterialExpressionConstant_0 = Constant() {r=0.02}
  MaterialExpressionAdd_3 = Add(A=Saturate(None=@n1.R)) {constB=0}
  n12 = Custom!(a=@MaterialExpressionConstant_0, b=@MaterialExpressionAdd_3) {
   code="return step(a,b);"
   outputType=CMOT_Float1
   inputs [
    {
     inputName=a
     input {expression=&self:MaterialExpressionConstant_0}
    }
    {
     inputName=b
     input {expression=&self:MaterialExpressionAdd_3}
    }
   ]
  }
  n13 = LinearInterpolate(A=@n8.Emissive Color, B=Multiply(A=@n8.Emissive Color, B=Multiply(A=@n10, B=@n11)), Alpha=@n12)
  n14 = ScalarParameter() {defaultValue=0.3, group=TwinklingLights, parameterName=BulbsRoughness}
  n15 = TextureSampleParameter2D() {
   parameterName=WireMask
   samplerType=SAMPLERTYPE_LinearColor
   texture=/BRCosmetics/Weapons/FORT_Melee/Pickaxe_Holiday_Lights_Male/Textures/Holiday_Lights_Wire_Mask
  }
  n16 = ScalarParameter() {defaultValue=1, group=TwinklingLights, parameterName=BulbsMetallic}
  n17 = ScalarParameter() {defaultValue=0.1, parameterName=BulbsBaseHue}
  n18 = Multiply(
   A=Desaturation(
    None=MaterialFunctionCall(Hue Shift Percentage=@n17, Texture=@n8.Base Color) {materialFunction=&HueShift}.Result
    Fraction=ScalarParameter() {defaultValue=0.5, parameterName=BulbsBaseDesat}
   )
   B=ScalarParameter() {defaultValue=2, parameterName=BulbsBaseInt}
  )
  n19 = SetMaterialAttributes!(
   MaterialAttributes=@n8.MaterialAttributes
   Emissive Color=Desaturation(
    None=MaterialFunctionCall(Hue Shift Percentage=@n9, Texture=@n13) {materialFunction=&HueShift}.Result
    Fraction=ScalarParameter() {parameterName=EmissiveDesat}
   )
   Roughness=LinearInterpolate(A=LinearInterpolate(A=@n8.Roughness, B=@n14, Alpha=@n12), B=ScalarParameter() {defaultValue=0.3, parameterName=WiresRoughness}, Alpha=@n15.G)
   Metallic=LinearInterpolate(A=LinearInterpolate(A=@n8.Metallic, B=@n16, Alpha=@n12), B=ScalarParameter() {defaultValue=0.5, parameterName=WiresMetallic}, Alpha=@n15.G)
   Base Color=LinearInterpolate(A=LinearInterpolate(A=@n8.Base Color, B=@n18, Alpha=@n12), B=ScalarParameter() {defaultValue=1, parameterName=WiresBaseColor}, Alpha=@n15.G)
   Specular=LinearInterpolate(A=@n8.Specular, B=ScalarParameter() {defaultValue=1, parameterName=WiresSpecular}, Alpha=@n15.G)
  ) {attributeSetTypes [
    #1
    #2
    #3
    #4
    #5
   ]}
  MaterialFunctionCall(In=@n19) {materialFunction=&MF_TechArtOverrides_accessories}
 ]
}
