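# Physics asset for the Holiday Lights (male) pickaxe, giving capsule bodies to the handle and its dangling hat, ribbon and light bones.
# Four three-bone ribbon chains, a hat chain with a locked tip, and a base with two hat-ribbon tails all constrain back to the handle.
# A lights base, free to rotate on the handle like the hat-ribbon base, carries narrowly limited light bones, with two-link chains for the eyes.
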PhysicsAsset /BRCosmetics/Weapons/FORT_Melee/Pickaxe_Holiday_Lights_Male/Meshes/Holiday_Lights_Male_Physics
bodies {
 Melee_TwoHanded_Handle [{shapeType=Capsule, length=74.6352, radius=25.6028, center=<0.2068 76.3391 33.2727>, rotation=<rot -0.618 -90 0.8031>}]
 dyn_hat_01 [{shapeType=Capsule, length=5.3668, radius=3.375, center=<1.238 0 0>}]
 dyn_hat_02 [{shapeType=Capsule, length=5.3668, radius=3.375, center=<1.641 0.0001 0>}]
 dyn_hat_ribbon_base [{shapeType=Capsule, length=9.273, radius=3.1863, center=<0.9396 -3.2224 3.3899>, rotation=<rot -50.8218 101.4114 -108.2944>}]
 dyn_hat_ribbon_l_01 [{shapeType=Capsule, length=0.75, radius=0.8968, center=<1.2042 0 0>, rotation=<rot 90 0 0>}]
 dyn_hat_ribbon_l_02 [{shapeType=Capsule, length=0.75, radius=0.8968, center=<1.2042 0 0>, rotation=<rot 90 0 0>}]
 dyn_hat_ribbon_r_01 [{shapeType=Capsule, length=0.75, radius=0.8968, center=<1.2042 0 0>, rotation=<rot 90 0 0>}]
 dyn_hat_ribbon_r_02 [{shapeType=Capsule, length=0.75, radius=0.8968, center=<1.2042 0 0>, rotation=<rot 90 0 0>}]
 dyn_lights_base [{shapeType=Capsule, length=19.977, radius=2.0774, center=<0.6608 0.7676 -27.9735>, rotation=<rot -1.3338 0 -1.581>}]
 dyn_lights_bk_l [{shapeType=Capsule, length=5.4115, radius=2.6006, center=<1.7837 -0.0001 0>, rotation=<rot 90 0 0>}]
 dyn_lights_bk_r [{shapeType=Capsule, length=5.4115, radius=2.6006, center=<1.7837 -0.0001 0>, rotation=<rot 90 0 0>}]
 dyn_lights_bk_r2 [{shapeType=Capsule, length=5.4115, radius=2.6006, center=<1.7837 -0.0001 0>, rotation=<rot 90 0 0>}]
 dyn_lights_eye_base [{shapeType=Capsule, length=5.4115, radius=2.6006, center=<0.7358 -0.0001 0>, rotation=<rot 90 0 0>}]
 dyn_lights_eye_base1 [{shapeType=Capsule, length=5.4115, radius=2.6006, center=<0.7358 -0.0001 0>, rotation=<rot 90 0 0>}]
 dyn_lights_eye_l [{shapeType=Capsule, length=4.6937, radius=4.2761, center=<-3.5012 -0.7063 -1.3867>, rotation=<rot -43.7948 -90 -7.5298>}]
 dyn_lights_eye_r [{shapeType=Capsule, length=4.6937, radius=4.2761, center=<-3.5012 -0.7063 -1.3867>, rotation=<rot -43.7948 -90 -7.5298>}]
 dyn_lights_mouth [{shapeType=Capsule, length=5.4115, radius=2.6006, center=<1.7837 -0.0001 0>, rotation=<rot 90 0 0>}]
 dyn_lights_top [{shapeType=Capsule, length=5.4115, radius=2.6006, center=<1.7837 -0.0001 0>, rotation=<rot 90 0 0>}]
 dyn_lights_top2 [{shapeType=Capsule, length=5.4115, radius=2.6006, center=<1.7837 -0.0001 0>, rotation=<rot 90 0 0>}]
 dyn_ribbon1_01 [{shapeType=Capsule, length=0.25, radius=1.6167, center=<1.7174 0 0.715>, rotation=<rot 90 -26.565 -116.565>}]
 dyn_ribbon1_02 [{shapeType=Capsule, length=1.6309, radius=1.6167, center=<1.2141 -0.0001 0>}]
 dyn_ribbon1_03 [{shapeType=Capsule, length=2.7559, radius=1.6167, center=<1.5283 0 0>}]
 dyn_ribbon2_01 [{shapeType=Capsule, length=0.2559, radius=1.6167, center=<1.5283 0 0.7151>, rotation=<rot 90 0 -90>}]
 dyn_ribbon2_02 [{shapeType=Capsule, length=1.5059, radius=1.6167, center=<1.9685 0 0>}]
 dyn_ribbon2_03 [{shapeType=Capsule, length=2.6309, radius=1.6167, center=<1.5283 0 0>}]
 dyn_ribbon3_01 [{shapeType=Capsule, length=3.8809, radius=1.6167, center=<1.5283 0 0>, rotation=<rot 90 -26.565 -116.565>}]
 dyn_ribbon3_02 [{shapeType=Capsule, length=7.6309, radius=1.6167, center=<1.9685 0 0>, rotation=<rot 90 0 -90>}]
 dyn_ribbon3_03 [{shapeType=Capsule, length=9.5059, radius=1.6167, center=<1.5283 0 0>, rotation=<rot 90 -26.565 -116.565>}]
 dyn_ribbon4_01 [{shapeType=Capsule, length=3.8809, radius=1.6167, center=<1.5283 0 0>, rotation=<rot 90 -26.565 -116.565>}]
 dyn_ribbon4_02 [{shapeType=Capsule, length=7.6309, radius=1.6167, center=<1.9685 0 0>, rotation=<rot 90 0 -90>}]
 dyn_ribbon4_03 [{shapeType=Capsule, length=9.5059, radius=1.6167, center=<1.5283 0 0>, rotation=<rot 90 -26.565 -116.565>}]
}
constraints [
 {bone1Name=dyn_ribbon3_01, bone2Name=Melee_TwoHanded_Handle, swing1LimitDegrees=5, swing2LimitDegrees=20, twistLimitDegrees=1}
 {bone1Name=dyn_ribbon3_02, bone2Name=dyn_ribbon3_01, swing1LimitDegrees=5, swing2LimitDegrees=20, twistLimitDegrees=1}
 {bone1Name=dyn_ribbon3_03, bone2Name=dyn_ribbon3_02, swing1LimitDegrees=5, swing2LimitDegrees=20, twistLimitDegrees=1}
 {bone1Name=dyn_ribbon4_03, bone2Name=dyn_ribbon4_02, swing1LimitDegrees=5, swing2LimitDegrees=20, twistLimitDegrees=1}
 {bone1Name=dyn_ribbon4_02, bone2Name=dyn_ribbon4_01, swing1LimitDegrees=5, swing2LimitDegrees=20, twistLimitDegrees=1}
 {bone1Name=dyn_ribbon4_01, bone2Name=Melee_TwoHanded_Handle, swing1LimitDegrees=5, swing2LimitDegrees=20, twistLimitDegrees=1}
 {bone1Name=dyn_ribbon2_01, bone2Name=Melee_TwoHanded_Handle, swing1LimitDegrees=35, swing2LimitDegrees=35, twistLimitDegrees=1}
 {bone1Name=dyn_ribbon2_02, bone2Name=dyn_ribbon2_01, swing1LimitDegrees=35, swing2LimitDegrees=35, twistLimitDegrees=1}
 {bone1Name=dyn_ribbon2_03, bone2Name=dyn_ribbon2_02, swing1LimitDegrees=35, swing2LimitDegrees=35, twistLimitDegrees=1}
 {bone1Name=dyn_ribbon1_01, bone2Name=Melee_TwoHanded_Handle, swing1LimitDegrees=35, swing2LimitDegrees=35, twistLimitDegrees=1}
 {bone1Name=dyn_ribbon1_02, bone2Name=dyn_ribbon1_01, swing1LimitDegrees=35, swing2LimitDegrees=35, twistLimitDegrees=1}
 {bone1Name=dyn_ribbon1_03, bone2Name=dyn_ribbon1_02, swing1LimitDegrees=35, swing2LimitDegrees=35, twistLimitDegrees=1}
 {
  bone1Name=dyn_lights_base
  bone2Name=Melee_TwoHanded_Handle
  swing1LimitDegrees=45
  swing1Motion=Free
  swing2LimitDegrees=45
  swing2Motion=Free
  twistLimitDegrees=45
  twistMotion=Free
 }
 {bone1Name=dyn_hat_01, bone2Name=Melee_TwoHanded_Handle, swing1LimitDegrees=10, swing2LimitDegrees=10, twistLimitDegrees=1}
 {
  bone1Name=dyn_hat_02
  bone2Name=dyn_hat_01
  swing1LimitDegrees=20
  swing1Motion=Locked
  swing2LimitDegrees=20
  swing2Motion=Locked
  twistLimitDegrees=1
  twistMotion=Locked
 }
 {bone1Name=dyn_hat_ribbon_r_01, bone2Name=dyn_hat_ribbon_base, swing1LimitDegrees=35, swing2LimitDegrees=35, twistLimitDegrees=1}
 {bone1Name=dyn_hat_ribbon_l_02, bone2Name=dyn_hat_ribbon_l_01, swing1LimitDegrees=35, swing2LimitDegrees=35, twistLimitDegrees=1}
 {bone1Name=dyn_hat_ribbon_l_01, bone2Name=dyn_hat_ribbon_base, swing1LimitDegrees=35, swing2LimitDegrees=35, twistLimitDegrees=1}
 {bone1Name=dyn_hat_ribbon_r_02, bone2Name=dyn_hat_ribbon_r_01, swing1LimitDegrees=35, swing2LimitDegrees=35, twistLimitDegrees=1}
 {
  bone1Name=dyn_hat_ribbon_base
  bone2Name=Melee_TwoHanded_Handle
  swing1LimitDegrees=45
  swing1Motion=Free
  swing2LimitDegrees=45
  swing2Motion=Free
  twistLimitDegrees=45
  twistMotion=Free
 }
 {bone1Name=dyn_lights_eye_base, bone2Name=dyn_lights_base, swing1LimitDegrees=10, swing2LimitDegrees=15, twistLimitDegrees=1}
 {bone1Name=dyn_lights_eye_base1, bone2Name=dyn_lights_base, swing1LimitDegrees=10, swing2LimitDegrees=15, twistLimitDegrees=1}
 {bone1Name=dyn_lights_bk_l, bone2Name=dyn_lights_base, swing1LimitDegrees=15, swing2LimitDegrees=30, twistLimitDegrees=1}
 {bone1Name=dyn_lights_bk_r, bone2Name=dyn_lights_base, swing1LimitDegrees=15, swing2LimitDegrees=30, twistLimitDegrees=1}
 {bone1Name=dyn_lights_bk_r2, bone2Name=dyn_lights_base, swing1LimitDegrees=15, swing2LimitDegrees=30, twistLimitDegrees=1}
 {bone1Name=dyn_lights_mouth, bone2Name=dyn_lights_base, swing1LimitDegrees=15, swing2LimitDegrees=30, twistLimitDegrees=1}
 {bone1Name=dyn_lights_top2, bone2Name=dyn_lights_base, swing1LimitDegrees=15, swing2LimitDegrees=30, twistLimitDegrees=1}
 {bone1Name=dyn_lights_top, bone2Name=dyn_lights_base, swing1LimitDegrees=15, swing2LimitDegrees=30, twistLimitDegrees=1}
 {bone1Name=dyn_lights_eye_l, bone2Name=dyn_lights_eye_base1, swing1LimitDegrees=10, swing2LimitDegrees=10, twistLimitDegrees=1}
 {bone1Name=dyn_lights_eye_r, bone2Name=dyn_lights_eye_base, swing1LimitDegrees=10, swing2LimitDegrees=10, twistLimitDegrees=1}
]
