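# Animation Blueprint for the male Holiday Lights pickaxe cosmetic, adding physics motion to its skeleton and setting a material flag for front-end display.
# The input pose gets a slight additive bone-space downward offset on a lower lights bone, then feeds a RigidBody node.
# That node simulates in base-bone space with an override physics asset, custom gravity and damped component-motion influence.
# On initialization, it sets IsInFrontEnd on a dynamic material instance from the front-end state.
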AnimBlueprint /BRCosmetics/Weapons/FORT_Melee/Pickaxe_Holiday_Lights_Male/Meshes/Holiday_Lights_Male_AnimBP
target_skeleton=&/Holiday_Lights_Male_Skeleton
graphs [
 AnimGraph {nodes [
   AnimGraphNode_SubInput_0 = Input Pose()
   n1 = "Transform (Modify) Bone - Bone: dny_lights_low_1"(ComponentPose=Local To Component(LocalPose=@AnimGraphNode_SubInput_0), Alpha=1, Translation="0,0,-0.1", Rotation="0,0,0", Scale="0,0,0") {
    rotationMode=BMM_Additive
    rotationSpace=BCS_BoneSpace
    scaleSpace=BCS_BoneSpace
    translationMode=BMM_Additive
    translationSpace=BCS_BoneSpace
    alphaBoolBlend {blendOption=Linear}
    boneToModify {boneName=dny_lights_low_1}
    scale=<0 0 0>
    translation {z=-0.1}
   }
   n2 = RigidBody(ComponentPose=@n1) {
    bClampLinearTranslationLimitToRefPose=true
    bOverrideWorldGravity=true
    overridePhysicsAsset=&/Holiday_Lights_Male_Physics
    simulationSpace=BaseBoneSpace
    alphaBoolBlend {blendOption=Linear}
    baseBoneRef {boneName=b_AR_Root}
    componentAppliedLinearAccClamp=<900 900 900>
    componentLinearAccScale=<0.3 0.4 0.3>
    componentLinearVelScale=<0.15 0.2 0.3>
    overrideWorldGravity {z=-980}
   }
   Output Pose(Result=Component To Local(ComponentPose=@n2)) {layerGroup=None}
  ]}
 EventGraph {nodes [
   Event BlueprintInitializeAnimation()
   > Parent: BlueprintInitializeAnimation() {is=title K2Node_CallParentFunction}
   > n1 = Get Dynamic Material Instance(SkeletalMeshComponent=GetPartSkeletalMeshComponent(), ElementIndex=1)
   > SetScalarParameterValue(self=@n1, ParameterName=IsInFrontEnd, Value=SelectFloat(A=1, B=0, bPickA=Get bIsInFrontEnd()))
   TryGetPawnOwner()
   Event BlueprintUpdateAnimation()
  ]}
]
variables []
